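# Unlit translucent Niagara mesh-particle material for the HighTower Date plasma energy ball.
# Emissive layers panning plasma shaped by a noise-distorted curve-atlas gradient, scrolling fog and a view/reflection-vector plasma, tinted by particle color.
# Dynamic parameter IsUltra blends green toward view-angle-dependent green-orange.
# Opacity combines inverted Fresnel edge fade, depth fade, gradient center fade and particle alpha, with Erode dissolving through plasma.
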Material /Game/Effects/Fort_Effects/GameplayPulginsAssets/Demon/M_HighTower_Date_Ball2
graph {
 properties {bUsedWithNiagaraMeshParticles=true, blendMode=BLEND_Translucent, shadingModel=MSM_Unlit, usedShadingModels=Unlit}
 nodes [
  n1 = Fresnel() {baseReflectFraction=0, exponent=0.2}
  Abs(None=@n1)
  MaterialFunctionCall() {materialFunction=&DebugTimeSine}
  n2 = ParticleColor()
  n3 = VectorParameter() {parameterName=CentColor2, defaultValue=<rgba 0.5213 1 0.0624 1>}
  n4 = MaterialFunctionCall("FadeSensitivity (optional)"=Constant() {r=3}) {materialFunction=&FadeBasedOnViewAngle}
  n5 = DynamicParameter() {paramNames [Erode,IsUltra,Param3,Param4]}
  n6 = LinearInterpolate(A=@n3.RGB, B=LinearInterpolate(A=VectorParameter() {defaultValue=<rgba 1 0.6271 0.0069 1>}.RGB, B=@n3.RGB, Alpha=@n4.Result), Alpha=@n5.G)
  n7 = MaterialFunctionCall() {materialFunction=&LinearGradient}
  n8 = AppendVector(A=@n7.UGradient, B=Power(Base=@n7.VGradient) {constExponent=1.5})
  n9 = Constant2Vector() {g=1, r=1}
  n10 = Time()
  n11 = ScalarParameter() {defaultValue=1, parameterName=PanSpeed}
  n12 = TextureSample(UVs=Panner(Coordinate=Multiply(A=@n8, B=@n9), Time=Multiply(A=@n10, B=@n11)) {speedY=-0.4}) {samplerType=SAMPLERTYPE_LinearColor, texture=/Game/Effects/Fort_Effects/Emotes/UltraLight/Texture/T_Plasma_A_C}
  n13 = TextureCoordinate()
  n14 = TextureSample(UVs=Panner() {speedX=0.5, speedY=0.5}) {samplerType=SAMPLERTYPE_Normal, texture=/Game/Characters/Player/Female/Medium/Bodies/F_MED_Bistro_Astronaut/FX/T_GaussianNoise_Normal}
  n15 = Divide(A=AppendVector(A=@n14.R, B=@n14.R)) {constB=9}
  n16 = CurveAtlasRowParameter(CurveTime=MaterialFunctionCall(UV Channel=Add(A=@n13, B=@n15)) {materialFunction=&LinearGradient}.VGradient) {
   atlas=&/CA_HighTower_Date
   curve=&/Curve_HighTower_Date_Ball
   parameterName=PlasmaShapeCurve
  }
  n17 = Multiply(A=Multiply(A=@n16.R), B=ScalarParameter() {defaultValue=2, parameterName=Param_1}) {constB=3}
  n18 = Saturate(None=LinearInterpolate(A=@n12.RGB, B=Subtract(A=@n12.RGB, B=@n17), Alpha=ScalarParameter() {defaultValue=1, parameterName=HalfOrFullPlasma}))
  n19 = VectorParameter() {parameterName=CentColor, defaultValue=<rgba 0.0394 0.4844 0.0261 1>}
  n20 = VectorParameter() {parameterName=Param_1, defaultValue=<rgba 1 0.0664 0 1>}
  n21 = Divide(A=LinearInterpolate(A=@n19.RGB, B=LinearInterpolate(A=@n20.RGB, B=@n19.RGB, Alpha=@n4.Result), Alpha=@n5.G))
  n22 = Panner(Coordinate=Multiply(A=TextureCoordinate(), B=Constant2Vector() {g=0.1, r=1}) {constB=2}) {speedY=-0.5}
  n23 = MaterialFunctionCall() {materialFunction=&LinearGradient}
  n24 = Multiply(
   A=Power(
    Base=TextureSample(UVs=@n22) {samplerType=SAMPLERTYPE_LinearGrayscale, texture=/Game/Environments/Hexmap/HexClouds/T_HexmapFog_01_Height02}.R
    Exp=Constant() {r=4}
   ) {constExponent=3}
   B=Saturate(None=Multiply(A=@n23.VGradient) {constB=5})
  )
  n25 = CameraVectorWS()
  n26 = ReflectionVectorWS()
  n27 = Multiply(A=ComponentMask(None=CrossProduct(A=@n25, B=@n26)) {b=true, g=true}, B=ScalarParameter() {defaultValue=0.4}) {constB=2}
  n28 = Constant2Vector() {g=1, r=-0.5}
  n29 = Multiply(
   A=MaterialFunctionCall() {materialFunction=&LinearGradient}.VGradient
   B=TextureSample(UVs=Panner(Coordinate=@n27, Speed=@n28) {speedX=1, speedY=1}) {samplerType=SAMPLERTYPE_LinearColor, texture=/Game/Effects/Fort_Effects/Emotes/UltraLight/Texture/T_Plasma_D_C}.RGB
  )
  n30 = Add(A=Multiply(A=@n6, B=Power(Base=@n18) {constExponent=1}), B=Add(A=LinearInterpolate(A=@n21, B=@n6, Alpha=@n24), B=Multiply(A=@n6, B=@n29))) {constB=0}
  MaterialFunctionCall(Emissive=Multiply(A=@n2.RGB, B=Divide(A=@n30))) {materialFunction=&MF_FX_EmissiveAdjustments}
  n31 = Multiply(
   A=MaterialFunctionCall() {materialFunction=&LinearGradient}.VGradient
   B=ScalarParameter() {defaultValue=1, parameterName=CenterFade_Width}
  )
  n32 = Multiply(A=OneMinus(None=@n1), B=ScalarParameter() {defaultValue=10, parameterName=FresnelEdgeFade}) {constB=10}
  n33 = LinearInterpolate(Alpha=DynamicParameter() {defaultValue=<rgba 0 0 1 1>, paramNames [Erode,IsUltra,Param3,Param4]}.Erode) {constB=4}
  n34 = Multiply(A=OneMinus(None=@n12.R), B=@n33)
  n35 = Subtract(A=DepthFade(Opacity=Saturate(None=@n32), FadeDistance=ScalarParameter() {defaultValue=100, parameterName=DepthFade}), B=@n34)
  Multiply(A=Power(Base=Saturate(None=@n31), Exp=ScalarParameter() {parameterName=CenterFade}), B=Multiply(A=@n2.A, B=Saturate(None=@n35)))
  Multiply(A=@n33)
  n36 = MaterialFunctionCall() {materialFunction=&LinearGradient}
  Multiply(A=Multiply(A=OneMinus(None=@n36.VGradient)) {constB=4}, B=@n34)
  Multiply()
  Multiply()
  OneMinus(None=@n18)
  TextureCoordinate()
  TextureCoordinate()
 ]
}
