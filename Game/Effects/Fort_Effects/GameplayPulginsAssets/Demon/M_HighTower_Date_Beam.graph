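# Translucent, unlit, two-sided material for HighTower Date (Demon) beam effects on Niagara ribbons.
# Particle color tints ColorA, which blends toward ColorB through IsUltra, LerpColor and axis-aligned fresnel.
# Opacity combines panning noise and plasma, curve-atlas end fading and cell-noise-modulated fresnel, with dynamic parameters driving erosion and growth along the ribbon.
# Depth fade and a pixel-depth clip fade soften intersections and near-camera areas.
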Material /Game/Effects/Fort_Effects/GameplayPulginsAssets/Demon/M_HighTower_Date_Beam
graph {
 properties {
  bAutomaticallySetUsageInEditor=false
  bUsedWithNiagaraRibbons=true
  blendMode=BLEND_Translucent
  shadingModel=MSM_Unlit
  twoSided=true
  usedShadingModels=Unlit
 }
 nodes [
  Add()
  n1 = ParticleColor()
  n2 = Multiply(A=VectorParameter() {parameterName=ColorA, defaultValue=<rgba 0.1191 1 0.0011 1>}.RGB, B=Constant() {r=5})
  n3 = Multiply(A=VectorParameter() {parameterName=ColorB, defaultValue=<rgba 1 0.0738 0 1>}.RGB, B=Constant() {r=1})
  n4 = LinearInterpolate(
   A=LinearInterpolate(A=@n2, B=@n3, Alpha=DynamicParameter() {parameterIndex=1, defaultValue=<rgba 1 1 1 0>, paramNames [X,Y,Z,LerpColor]}.LerpColor) {constAlpha=0}
   B=@n3
   Alpha=MaterialFunctionCall(Axis=NamedRerouteUsage(), Exponent=Constant() {r=3}) {materialFunction=&AxisAlignedFresnel}.Fresnel
  )
  n5 = DynamicParameter() {defaultValue=<rgba 1 1 1 0>, paramNames [Erode,Param2,IsUltra,Grow]}
  MaterialFunctionCall(Emissive=Multiply(A=@n1.RGB, B=LinearInterpolate(A=@n2, B=@n4, Alpha=@n5.IsUltra))) {materialFunction=&MF_FX_EmissiveAdjustments}
  MaterialFunctionCall() {materialFunction=&DebugTimeSine}
  n6 = TextureCoordinate() {coordinateIndex=1}
  n7 = DynamicParameter() {defaultValue=<rgba 1 1 1 0>, paramNames [Erode,Param2,IsUltra,Grow]}
  n8 = Multiply(A=Subtract(
    A=MaterialFunctionCall(UV Channel=@n6) {materialFunction=&LinearGradient}.UGradient
    B=OneMinus(None=@n7.Grow)
   )) {constB=5}
  n9 = CurveAtlasRowParameter(CurveTime=MaterialFunctionCall(UV Channel=TextureCoordinate() {coordinateIndex=1}) {materialFunction=&LinearGradient}.UGradient) {
   atlas=&/CA_HighTower_Date
   curve=&/Curve_HighTower_Date_Ball
  }
  n10 = Multiply(A=@n9.B)
  n11 = ScalarParameter() {defaultValue=1, parameterName=FadeBeamHeadAndTail}
  n12 = LinearInterpolate(B=Constant() {r=0.1}, Alpha=DynamicParameter() {defaultValue=<rgba 1 1 1 0>, paramNames [Erode,Param2,IsUltra,Grow]}.Erode) {constA=0.4, constB=0.4}
  n13 = Multiply(A=ScalarParameter() {defaultValue=1, parameterName=Tile}) {constB=4}
  n14 = Panner(
   Coordinate=Multiply(A=TextureCoordinate(), B=AppendVector(A=@n12, B=@n13))
   Speed=AppendVector(A=ScalarParameter() {defaultValue=2, parameterName=PanSpeedU}, B=Constant() {r=1})
  ) {speedX=3, speedY=1}
  n15 = Multiply(
   A=LinearInterpolate(Alpha=TextureSample(UVs=@n14) {texture=/Game/Effects/Fort_Effects/Textures/Tile/T_Noise_03}.R) {constA=0.1}
   B=TextureSample(UVs=Divide(A=@n14) {constB=1.5}) {samplerType=SAMPLERTYPE_LinearColor, texture=/Game/Effects/Fort_Effects/Emotes/UltraLight/Texture/T_Plasma_D_C}.R
  )
  n16 = Constant() {r=2}
  n17 = Subtract(A=Multiply(A=LinearInterpolate(B=@n10, Alpha=@n11) {constA=1}, B=Multiply(A=@n15, B=@n16) {constB=3}), B=@n7.Erode)
  n18 = TextureCoordinate() {uTiling=0.5, vTiling=0.5}
  n19 = Power(Base=MaterialFunctionCall(UV Channel=Sine(None=@n18)) {materialFunction=&LinearGradient}.VGradient) {constExponent=5}
  n20 = DynamicParameter() {parameterIndex=1, defaultValue=<rgba 1 1 1 0>, paramNames [X,Y,Z,LerpColor]}
  n21 = NamedRerouteDeclaration(None=MaterialFunctionCall(R=@n20.X, G=@n20.X, B=@n20.X) {materialFunction=&AppendMany}.RGB) {name=AxisAlignedFresnelInput, nodeColor=<rgba 1 0 0.0143 1>}
  n22 = ScalarParameter() {defaultValue=1, parameterName=AxisFresnelExp}
  n23 = LinearInterpolate(Alpha=Clamp(None=MaterialFunctionCall(Axis=@n21, Exponent=@n22) {materialFunction=&AxisAlignedFresnel}.Fresnel)) {constB=5}
  n24 = Constant() {r=0.5}
  n25 = Constant() {r=2}
  n26 = LinearInterpolate(
   A=Multiply(A=@n24, B=ScalarParameter() {defaultValue=40, parameterName=FresnelExp}) {constB=3}
   B=@n24
   Alpha=TextureSample(UVs=Multiply(A=@n14, B=@n25) {constB=2}) {samplerType=SAMPLERTYPE_Grayscale, texture=/Game/Weapons/FORT_Melee/Pickaxe_Cube_Ninja_Male/FX/T_CellNoise}.R
  )
  n27 = Multiply(
   A=Multiply(A=Saturate(None=@n8), B=Saturate(None=@n17))
   B=LinearInterpolate(
    A=Multiply(A=@n19, B=@n23)
    B=MaterialFunctionCall(Axis=@n21, Exponent=@n26) {materialFunction=&AxisAlignedFresnel}.Fresnel
    Alpha=ScalarParameter() {defaultValue=1, parameterName=UseFresnel}
   ) {constA=1}
  )
  n28 = Subtract(A=PixelDepth().R, B=ScalarParameter() {defaultValue=75, parameterName=Clip Plane Distance})
  n29 = ScalarParameter() {defaultValue=500, parameterName=Soften}
  Multiply(
   A=DepthFade(Opacity=Multiply(A=@n1.A, B=@n27), FadeDistance=ScalarParameter() {defaultValue=150, parameterName=DepthFadeDis})
   B=Saturate(None=Divide(A=@n28, B=@n29))
  )
  ScalarParameter() {defaultValue=0.4, parameterName=Param_1}
 ]
}
