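# Translucent unlit Niagara sprite material for a soft, flickering, wobbling dot, filed under the Demon gameplay effects.
# Soft noise, scrolled over time and offset by particle random, distorts the UVs of a sphere mask.
# A DebugTimeSine with per-particle randomized speed, driven by FlickerSpeed, pulses that mask's radius.
# A fixed circular mask clips the result, which is then multiplied by particle color alpha.
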Material /Game/Effects/Fort_Effects/GameplayPulginsAssets/Demon/M_HighTower_Date_Dots
graph {
 properties {
  bAutomaticallySetUsageInEditor=false
  bUsedWithNiagaraSprites=true
  blendMode=BLEND_Translucent
  shadingModel=MSM_Unlit
  usedShadingModels=Unlit
 }
 nodes [
  Constant()
  n1 = TextureCoordinate()
  n2 = Constant() {r=0.5}
  n3 = Multiply(A=TextureCoordinate()) {constB=0.5}
  n4 = LinearInterpolate(Alpha=ParticleRandom())
  n5 = TextureSample(UVs=Add(A=Add(A=@n3, B=@n4), B=Time())) {texture=/Game/Effects/Fort_Effects/Emotes/UltraLight/Texture/T_SoftNoise_BC7}
  n6 = Add(A=@n1, B=Divide(A=ComponentMask(None=@n5.RGB) {g=true, r=true}) {constB=3})
  n7 = ScalarParameter() {defaultValue=6, parameterName=FlickerSpeed}
  n8 = Divide(A=@n7) {constB=1.5}
  n9 = ParticleRandom()
  n10 = LinearInterpolate(Alpha=MaterialFunctionCall(Speed=LinearInterpolate(A=@n7, B=@n8, Alpha=@n9) {constA=5, constB=6}) {materialFunction=&DebugTimeSine}.Result) {constA=0.25, constB=0.5}
  Multiply(
   A=Multiply(A=SphereMask(A=@n1, B=@n2, Radius=@n2) {hardnessPercent=80}, B=SphereMask(A=@n6, B=@n2, Radius=@n10) {hardnessPercent=80})
   B=ParticleColor().A
  )
 ]
}
