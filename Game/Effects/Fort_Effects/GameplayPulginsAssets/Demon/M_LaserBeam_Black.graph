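# Masked, two-sided unlit Niagara ribbon material for a dark-cored laser beam effect.
# Cell noise, panned and warped by a swirl flowmap, is eroded via dynamic parameters and shaped by a UV1 curve-atlas gradient, axis-aligned fresnel, particle alpha and a Grow wipe.
# That mask is temporally dithered and faded near the camera by pixel depth.
# Emissive blends bright orange OutsideColor and near-black InsideColor using fresnel-weighted noise and face sign.
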Material /Game/Effects/Fort_Effects/GameplayPulginsAssets/Demon/M_LaserBeam_Black
graph {
 properties {
  bAutomaticallySetUsageInEditor=false
  bUsedWithNiagaraRibbons=true
  blendMode=BLEND_Masked
  shadingModel=MSM_Unlit
  twoSided=true
  usedShadingModels=Unlit
 }
 nodes [
  n1 = CurveAtlasRowParameter(CurveTime=MaterialFunctionCall(UV Channel=TextureCoordinate() {coordinateIndex=1}) {materialFunction=&LinearGradient}.UGradient) {
   atlas=&/CA_HighTower_Date
   curve=&/Curve_HighTower_Date_Ball
   parameterName=Param_1
  }
  n2 = Multiply(A=@n1.B)
  n3 = ScalarParameter() {defaultValue=1, parameterName=Tile}
  n4 = Multiply(
   A=TextureCoordinate()
   B=MaterialFunctionCall(X=Multiply(A=@n3) {constB=0.2}, Y=Constant()) {materialFunction=&MakeFloat2}.Result
  )
  n5 = Constant() {r=0.2}
  n6 = Constant() {r=3}
  n7 = Panner(
   Coordinate=Multiply(
    A=TextureCoordinate()
    B=MaterialFunctionCall(X=@n5, Y=@n6) {materialFunction=&MakeFloat2}.Result
   )
   Time=NamedRerouteUsage()
   Speed=Multiply(A=NamedRerouteUsage(), B=Constant() {r=0.5})
  ) {speedX=0.5}
  n8 = Divide(
   A=ComponentMask(None=TextureSample(UVs=@n7) {samplerType=SAMPLERTYPE_Normal, texture=/Game/Global/Textures/Default/Flowmaps/T_SwirlFlow_02_N}.RGB) {g=true, r=true}
   B=Constant() {r=20}
  ) {constB=8}
  n9 = NamedRerouteUsage()
  n10 = Constant() {r=0.7}
  n11 = TextureSample(UVs=Panner(Coordinate=Add(A=@n4, B=@n8), Time=NamedRerouteUsage(), Speed=Multiply(A=@n9, B=@n10)) {speedX=0.7}) {samplerType=SAMPLERTYPE_Grayscale, texture=/Game/Weapons/FORT_Melee/Pickaxe_Cube_Ninja_Male/FX/T_CellNoise}
  n12 = Power(Base=@n11.R, Exp=ScalarParameter() {defaultValue=4, parameterName=Param_2})
  n13 = DynamicParameter() {defaultValue=<rgba 0 1 1 1>, paramNames [Erode,Param2,Param3,Grow]}
  n14 = Multiply(A=LinearInterpolate(A=Constant() {r=0.05}, B=Constant() {r=1}, Alpha=@n13.Erode), B=Constant() {r=1})
  n15 = Multiply(A=Saturate(None=Subtract(A=@n12, B=@n14)), B=Multiply(A=ParticleColor().A, B=Constant() {r=2}) {constB=2})
  n16 = NamedRerouteUsage()
  n17 = ScalarParameter() {defaultValue=3, parameterName=Fresnel}
  n18 = Subtract(
   A=MaterialFunctionCall(UV Channel=TextureCoordinate() {coordinateIndex=1}) {materialFunction=&LinearGradient}.UGradient
   B=OneMinus(None=@n13.Grow)
  )
  n19 = Multiply(
   A=Multiply(
    A=Multiply(A=@n2, B=@n15)
    B=MaterialFunctionCall(Axis=@n16, Exponent=@n17) {materialFunction=&AxisAlignedFresnel}.Fresnel
   )
   B=Saturate(None=Multiply(A=@n18) {constB=5})
  )
  DepthFade(Opacity=@n19, FadeDistance=ScalarParameter() {defaultValue=500, parameterName=DepthFade})
  MaterialFunctionCall() {materialFunction=&DebugTimeSine}
  n20 = VectorParameter() {parameterName=OutsideColor, defaultValue=<rgba 1 0.3162 0.0401 1>}
  n21 = Multiply(A=@n20.RGB, B=ScalarParameter() {defaultValue=10, parameterName=Param_4}) {constB=3}
  n22 = Multiply(
   A=MaterialFunctionCall(Axis=NamedRerouteUsage(), Exponent=Constant() {r=4}) {materialFunction=&AxisAlignedFresnel}.Fresnel
   B=@n11.R
  )
  n23 = VectorParameter() {parameterName=InsideColor, defaultValue=<rgba 0.0174 0.0089 0 1>}
  MaterialFunctionCall(Emissive=LinearInterpolate(A=LinearInterpolate(A=@n20.RGB, B=@n21, Alpha=@n22), B=LinearInterpolate(B=@n23.RGB, Alpha=@n22), Alpha=TwoSidedSign()) {constB=0}) {materialFunction=&MF_FX_EmissiveAdjustments}
  n24 = Subtract(A=PixelDepth().R, B=ScalarParameter() {defaultValue=150, parameterName=Clip Plane Distance})
  n25 = ScalarParameter() {defaultValue=500, parameterName=Soften}
  Multiply(
   A=MaterialFunctionCall(Alpha Threshold=@n19, Random=Constant() {r=1}) {materialFunction=&DitherTemporalAA_Fortnite}.Result
   B=Saturate(None=Divide(A=@n24, B=@n25))
  )
  NamedRerouteDeclaration(None=MaterialFunctionCall(X=ScalarParameter() {defaultValue=1, parameterName=PanSpeedU}, Y=ScalarParameter() {defaultValue=0.2, parameterName=PanSpeedV}) {materialFunction=&MakeFloat2}.Result) {name=PanSpeed, nodeColor=<rgba 0 0.6953 1 1>}
  n26 = DynamicParameter() {parameterIndex=1, paramNames [X,Y,Z,Param4]}
  NamedRerouteDeclaration(None=MaterialFunctionCall(R=@n26.X, G=@n26.X, B=@n26.X) {materialFunction=&AppendMany}.RGB) {name=AxisAlignedFresnelInput, nodeColor=<rgba 1 0 0.0143 1>}
  n27 = Time()
  n28 = ScalarParameter() {defaultValue=3, parameterName=PanSpeed}
  NamedRerouteDeclaration(None=Add(A=Multiply(A=@n27, B=@n28) {constB=0}, B=ScalarParameter() {parameterName=TimeOffset})) {name=TimeInput, nodeColor=<rgba 1 0.034 0 1>}
 ]
}
