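# Translucent unlit Niagara sprite material for the LodeSwan power-up aura.
# Emissive blends particle color with an optional hue-shifted curve-atlas gradient sampled through a sphere mask.
# Opacity multiplies scrolling distorted noise by a noise-warped aura mask, an optional top mask and dissolve breakup, then applies particle alpha and view-angle fade.
# A dynamic parameter drives noise time; depth fade is skipped on forward and mobile paths.
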Material /Game/Effects/Fort_Effects/GameplayPulginsAssets/Demon/M_LodeSwan_PowerUp_Aura
graph {
 properties {bUsedWithNiagaraSprites=true, blendMode=BLEND_Translucent, pixelDepthOffsetMode=PDOM_Legacy, shadingModel=MSM_Unlit, usedShadingModels=Unlit}
 nodes [
  n1 = ParticleColor()
  n2 = TextureCoordinate()
  n3 = Constant2Vector() {g=0.65, r=0.5}
  n4 = CurveAtlasRowParameter(CurveTime=Power(Base=SphereMask(A=@n2, B=@n3) {attenuationRadius=0.7, hardnessPercent=0})) {
   atlas=/Game/Athena/LinearColorCurves/A_Atlas_Turtleneck_Glider
   curve=/Game/Effects/Fort_Effects/Emotes/Energize_Ultra/LV4/Energize_Ultra_color_Curve
   defaultValue=23
  }
  n5 = MaterialFunctionCall(Hue Shift Percentage=ScalarParameter() {parameterName=HueShift}, Texture=@n4) {materialFunction=&HueShift}
  MaterialFunctionCall(Emissive=LinearInterpolate(A=@n1.RGB, B=Multiply(A=@n5.Result, B=@n1.RGB), Alpha=ScalarParameter() {parameterName=Use gradient color})) {materialFunction=&MF_FX_EmissiveAdjustments}
  n6 = TextureCoordinate() {vTiling=1.5}
  n7 = ComponentMask(None=TextureSample(UVs=Panner(Coordinate=@n6) {speedY=0.5}) {samplerType=SAMPLERTYPE_Normal, texture=&T_Weap_Distort_N_02}.RGB) {g=true, r=true}
  n8 = OneMinus(None=ComponentMask(None=TextureCoordinate()) {g=true})
  n9 = TextureCoordinate() {uTiling=0.7, vTiling=0.8}
  n10 = Constant2Vector() {g=0.5}
  n11 = Add(
   A=Multiply(A=Multiply(A=@n7, B=@n8)) {constB=0.3}
   B=MaterialFunctionCall(XY=Add(A=@n9, B=@n10)) {materialFunction=&Swizzle}.YX
  )
  n12 = DynamicParameter() {paramNames [Param1, Time speed, Param3, Param4]}
  n13 = ComponentMask(None=VectorParameter() {parameterName=Color noise speed, defaultValue=<rgba 1 0 0 1>}.RGB) {g=true, r=true}
  n14 = TextureSample() {samplerType=SAMPLERTYPE_LinearGrayscale, texture=/Engine/Functions/Engine_MaterialFunctions02/ExampleContent/Textures/LowResBlurredNoise}
  n15 = LinearInterpolate(
   A=TextureSample(UVs=Panner(Coordinate=@n11, Time=@n12.Time speed, Speed=@n13) {speedX=1}) {samplerType=SAMPLERTYPE_LinearColor, texture=&T_NoisePack_LinearColor}.G
   B=@n14.R
   Alpha=Constant() {r=0.5}
  ) {constAlpha=0.2}
  n16 = ScalarParameter() {defaultValue=3, parameterName=Noise power}
  n17 = Add(
   A=TextureCoordinate() {vTiling=2}
   B=ComponentMask(None=VectorParameter() {parameterName=UV offset, defaultValue=<rgba 0 0 0 1>}.RGB) {g=true, r=true}
  )
  n18 = DynamicParameter() {paramNames [Param1, Time speed, Param3, Param4]}
  n19 = LinearInterpolate(Alpha=TextureSample(UVs=Panner(Coordinate=@n17, Time=@n18.Time speed) {speedY=2.3}) {samplerType=SAMPLERTYPE_Normal, texture=&T_Weap_Distort_N_02}.R)
  n20 = TextureCoordinate()
  n21 = Power(Base=OneMinus(None=ComponentMask(None=@n20) {g=true}), Exp=ScalarParameter() {defaultValue=0.8, parameterName=bottom}) {constExponent=0.8}
  n22 = Add(
   A=TextureCoordinate()
   B=Multiply(A=Multiply(A=@n19, B=@n21), B=ScalarParameter() {defaultValue=0.5, parameterName=Noise intensity}) {constB=0.5}
  )
  n23 = SphereMask(A=TextureCoordinate(), B=Constant2Vector() {g=0.7, r=0.5}) {attenuationRadius=0.5, hardnessPercent=0}
  n24 = Multiply(
   A=Multiply(
    A=Power(Base=@n15, Exp=@n16) {constExponent=3}
    B=TextureSample(UVs=@n22) {samplerSource=SSM_Clamp_WorldGroupSettings, texture=/Game/Effects/Fort_Effects/Emotes/Energize_Max/LV3/T_Energize_Max_Aura_LV3_mask}.R
   )
   B=LinearInterpolate(A=Constant() {r=1}, B=Power(Base=@n23) {constExponent=3}, Alpha=ScalarParameter() {parameterName=Use top mask})
  )
  n25 = TextureCoordinate()
  n26 = Constant() {r=7}
  n27 = Multiply(A=Panner(Coordinate=@n25) {speedX=0.1, speedY=0.2}, B=@n26) {constB=2}
  n28 = Multiply(A=Panner(Coordinate=@n25) {speedX=-0.05, speedY=0.2}, B=Multiply(A=@n26) {constB=1.5}) {constB=1.5}
  n29 = Multiply(
   A=OneMinus(None=TextureSample(UVs=@n27) {texture=/Game/CharClasses/Outlander/Gadgets/AntiMaterialCharge/Materials/T_AntiMaterialCharge_Dissolve_01}.R)
   B=OneMinus(None=TextureSample(UVs=@n28) {texture=&T_AntiMaterialCharge_Dissolve_01}.R)
  )
  n30 = LinearInterpolate(Alpha=Saturate(None=Subtract(A=@n29) {constB=0.1})) {constA=0.8}
  n31 = Multiply(
   A=Multiply(A=@n1.A, B=Multiply(A=@n24, B=@n30))
   B=MaterialFunctionCall("FadeSensitivity (optional)"=Constant() {r=2}) {materialFunction=&FadeBasedOnViewAngle}.Result
  )
  n32 = DepthFade(Opacity=@n31) {fadeDistanceDefault=30}
  ShadingPathSwitch(Default=@n32, Deferred=@n32, Forward=@n31, Mobile=@n31)
 ]
}
