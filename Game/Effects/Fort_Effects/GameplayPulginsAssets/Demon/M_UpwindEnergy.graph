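# Translucent unlit Niagara sprite and mesh material for an upwind energy effect in the Demon effects set.
# Two noise layers distort a plasma texture with power-curved V, all timed by a dynamic parameter, and the refined result blends base to particle-tinted main color.
# Opacity math combines an edge mask, axis-aligned fresnel, particle alpha and depth fade.
# A separate branch turns the first noise into a scaled, transformed vector offset.
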Material /Game/Effects/Fort_Effects/GameplayPulginsAssets/Demon/M_UpwindEnergy
graph {
 properties {
  bAutomaticallySetUsageInEditor=false
  bUsedWithNiagaraMeshParticles=true
  bUsedWithNiagaraSprites=true
  blendMode=BLEND_Translucent
  shadingModel=MSM_Unlit
  usedShadingModels=Unlit
 }
 nodes [
  n1 = ParticleColor()
  n2 = Multiply(A=VectorParameter() {parameterName=MainC, defaultValue=<rgba 3 0.55 0.06 1>}.RGB, B=@n1.RGB)
  n3 = TextureObjectParameter() {
   group=MainT
   parameterName=MainT_Texture
   samplerType=SAMPLERTYPE_LinearGrayscale
   sortPriority=0
   texture=/Game/Effects/Fort_Effects/Emotes/UltraLight/Texture/T_Plasma_B_Greyscale
  }
  n4 = TextureObjectParameter() {
   group=MainB
   parameterName=MainB_Texture
   samplerType=SAMPLERTYPE_LinearGrayscale
   sortPriority=0
   texture=&T_SimplePerlin_Greyscale
  }
  n5 = VectorParameter() {group=MainB, parameterName=MainB_UVControl, channelNames {a=VPan, b=UPan, g=VScale, r=UScale}, defaultValue=<rgba 1 0.456 0 0.8038>}
  n6 = DynamicParameter() {defaultValue=<rgba 50 1 1 1>}
  n7 = VectorParameter() {
   group=MainB
   parameterName=MainB_NoiseControl
   sortPriority=60
   channelNames {a=VStrength, b=UStrength, g=VOffset, r=UOffset}
   defaultValue=<rgba -0.5 -0.5 0 0>
  }
  n8 = VectorParameter() {group=MainA, parameterName=MainA_UVControl, channelNames {a=VPan, b=UPan, g=VScale, r=UScale}, defaultValue=<rgba 2 0.7 0 0.5>}
  n9 = MaterialFunctionCall(
   Texture=TextureObjectParameter() {group=MainA, parameterName=MainA_Texture, sortPriority=0, texture=/Game/Items/Traps/Textures/T_PerlinRGB_01_M}
   UVControl_RGB=@n8
   UVControl_A=@n8
   "[Override]TimeInput"=@n6.R
  ) {materialFunction=&FXCommon_TextureSample}
  n10 = VectorParameter() {
   group=MainA
   parameterName=MainA_NoiseControl
   sortPriority=60
   channelNames {a=VStrength, b=UStrength, g=VOffset, r=UOffset}
   defaultValue=<rgba -0.5 -0.5 0.07 0.1>
  }
  n11 = Add(
   A=MaterialFunctionCall!(
    NoiseValue=MaterialFunctionCall(Texture=@n4, UVControl_RGB=@n5, UVControl_A=@n5, "[Override]TimeInput"=@n6.R) {materialFunction=&FXCommon_TextureSample}.RG
    Control_RGB=@n7
    Control_A=@n7
   ) {materialFunction=/Game/Effects/Niagara/MasterMaterials/Functions/Common/FXCommon_NoiseControl}.NoiseValue
   B=MaterialFunctionCall(NoiseValue=@n9.RG, Control_RGB=@n10, Control_A=@n10) {materialFunction=&FXCommon_NoiseControl}.NoiseValue
  )
  n12 = VectorParameter() {group=MainT, parameterName=MainT_UVControl, channelNames {a=VPan, b=UPan, g=VScale, r=UScale}, defaultValue=<rgba 2 0.3 0 0.2>}
  n13 = MaterialFunctionCall(Float2=TextureCoordinate()) {materialFunction=&BreakOutFloat2Components}
  n14 = Power(Base=OneMinus(None=@n13.G), Exp=ScalarParameter() {defaultValue=0.3, parameterName=EX}) {constExponent=0.5}
  n15 = MaterialFunctionCall(
   Texture=@n3
   AdditionalUVOffset=@n11
   UVControl_RGB=@n12
   UVControl_A=@n12
   "[Override]TimeInput"=@n6.R
   "[Override]UVInput"=MaterialFunctionCall(X=@n13.R, Y=OneMinus(None=@n14)) {materialFunction=&MakeFloat2}.Result
  ) {materialFunction=&FXCommon_TextureSample}
  n16 = VectorParameter() {parameterName=Refine, defaultValue=<rgba 5 0 0 0>}
  n17 = TextureCoordinate()
  n18 = MaterialFunctionCall(
   Texture=@n15.R
   RGB=LinearInterpolate(
    A=@n16.RGB
    B=VectorParameter() {parameterName=Refine4, defaultValue=<rgba 0.5 1 1 0>}.RGB
    Alpha=MaterialFunctionCall(Float2=@n17) {materialFunction=&BreakOutFloat2Components}.G
   )
   A=@n16.A
  ) {materialFunction=&MF_FX_Refine}
  MaterialFunctionCall(Emissive=LinearInterpolate(A=VectorParameter() {parameterName=BaseC, defaultValue=<rgba 0.09 0.045 0.045 1>}.RGB, B=Multiply(A=@n2, B=@n18.Result), Alpha=@n18.Result)) {materialFunction=&MF_FX_EmissiveAdjustments}
  Multiply(A=ScalarParameter() {defaultValue=38.7596, parameterName=FWD}, B=Transform(None=Subtract(A=@n9.RGB) {constB=0.5}))
  n19 = Add(A=TextureCoordinate(), B=@n11)
  n20 = VectorParameter() {group=Msk, parameterName=Msk_EdgeMask_InvertControl, channelNames {a="", b=Power, g=Multiply, r=Offset}, defaultValue=<rgba 0 5 2 1>}
  n21 = VectorParameter() {group=Msk, parameterName=Msk_EdgeMask_Control, channelNames {a="", b=Power, g=Multiply, r=Offset}, defaultValue=<rgba 0 1 4 1>}
  n22 = VectorParameter() {parameterName=Refine2, defaultValue=<rgba 1 15 1 0>}
  n23 = Multiply(
   A=MaterialFunctionCall(Texture=@n15.R, RGB=@n22.RGB, A=@n22.RGB) {materialFunction=&MF_FX_Refine}.Result
   B=MaterialFunctionCall(Axis=Constant3Vector() {constant=<rgba 0 0 1 0>}, Exponent=ScalarParameter() {defaultValue=5, parameterName=EXP}) {materialFunction=&AxisAlignedFresnel}.Fresnel
  )
  n24 = Multiply(
   A=MaterialFunctionCall!(
    UorV=MaterialFunctionCall(Float2=@n19) {materialFunction=&BreakOutFloat2Components}.G
    Invert_EdgeControl=@n20
    NoInvert_EdgeControl=@n21
   ) {materialFunction=/Game/Effects/Niagara/MasterMaterials/Functions/Common/FXCommon_EdgeMask_Ex}.Mask
   B=@n23
  )
  n25 = ScalarParameter() {defaultValue=100, parameterName=DF}
  Multiply(A=Multiply(A=@n1.A, B=Saturate(None=@n24)), B=Power(Base=DepthFade(FadeDistance=@n25)))
  Saturate(None=@n23)
 ]
}
